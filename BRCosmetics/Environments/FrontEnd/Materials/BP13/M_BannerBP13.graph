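# Subsurface-shaded banner cloth material for the BP13 front end.
# A banner mask's red channel blends two color pairs, selected by a Chrome parameter, and drives metallic and roughness-like blends.
# Its green channel feeds specular and, with contrast, multiplies color and a tiled cloth pattern for subsurface brightness.
# The banner normal is blended separately with the shared Cloth function's normal and with a tiled flake normal.
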Material /BRCosmetics/Environments/FrontEnd/Materials/BP13/M_BannerBP13
graph {
 properties {pixelDepthOffsetMode=PDOM_Legacy, shadingModel=MSM_Subsurface, usedShadingModels=Subsurface}
 nodes [
  n1 = TextureSample() {samplerType=SAMPLERTYPE_Masks, texture=/BRCosmetics/Environments/FrontEnd/Textures/BP13/T_cloth_banner_M}
  LinearInterpolate(A=ScalarParameter() {defaultValue=-0.2476, parameterName=C1}, B=ScalarParameter() {defaultValue=2.8045, parameterName=C2}, Alpha=@n1.R)
  n2 = Saturate(None=@n1.R)
  LinearInterpolate(B=ScalarParameter() {defaultValue=1, parameterName=Metallic}, Alpha=@n2)
  n3 = TextureSample(UVs=Multiply(A=TextureCoordinate(), B=ScalarParameter() {defaultValue=1, parameterName=ScaleClothPattern})) {texture=/Game/Characters/Player/Male/Medium/Bodies/M_MED_Martial_Artist/Textures/cloth}
  LinearInterpolate(
   A=ScalarParameter() {defaultValue=1.3387, parameterName=R3}
   B=LinearInterpolate(A=ScalarParameter() {defaultValue=0.3905, parameterName=R1}, B=ScalarParameter() {defaultValue=0.2, parameterName=R2}, Alpha=@n2)
   Alpha=@n3.R
  )
  n4 = TextureSample() {samplerType=SAMPLERTYPE_Normal, texture=/BRCosmetics/Environments/FrontEnd/Textures/BP13/T_cloth_banner_N}
  MaterialFunctionCall(
   BaseNormal=MaterialFunctionCall!(
    UVs
    Pan Speed
    WPO_ClothOffsetDirection
    WPO_ClothNormalIntensity
    WPO_ClothOffsetIntensity
    "Motion Mask (Defaults to the Red Vertex Color)"
    WPO_ClothWPOWindTextureOffset
    Pan Direction
   ) {materialFunction=/Game/Packages/Fortress_SharedMaterials/Base_Material_Functions/Cloth}.Normal
   AdditionalNormal=@n4.RGB
  ) {materialFunction=&BlendAngleCorrectedNormals}
  n5 = TextureCoordinate()
  n6 = ScalarParameter() {defaultValue=7.9483, parameterName=Flackes}
  MaterialFunctionCall(
   BaseNormal=@n4.RGB
   AdditionalNormal=TextureSample(UVs=Multiply(A=@n5, B=@n6)) {samplerType=SAMPLERTYPE_Normal, texture=/Game/Environments/Sets/Neo/Vehicles/Textures/T_Flakes_N}.RGB
  ) {materialFunction=&BlendAngleCorrectedNormals}
  n7 = LinearInterpolate(
   A=LinearInterpolate(
    A=VectorParameter() {parameterName=Color, defaultValue=<rgba 0.5 0.2764 0.0911 1>}.RGB
    B=VectorParameter() {parameterName=Color2, defaultValue=<rgba 7 2.7861 0.0559 1>}.RGB
    Alpha=@n2
   )
   B=LinearInterpolate(
    A=VectorParameter() {parameterName=Color4, defaultValue=<rgba 0.2981 0.9745 1.5 1>}.RGB
    B=VectorParameter() {parameterName=Color3, defaultValue=<rgba 6 6 6 1>}.RGB
    Alpha=@n2
   )
   Alpha=ScalarParameter() {defaultValue=1, parameterName=Chrome}
  )
  n8 = Power(Base=@n1.G, Exp=ScalarParameter() {defaultValue=3, parameterName=Contrast})
  Multiply(A=Multiply(A=@n3.R, B=Multiply(A=@n7, B=@n8)), B=ScalarParameter() {defaultValue=0.4, parameterName=SSSBirghtness})
  Multiply(A=@n1.G, B=ScalarParameter() {defaultValue=2, parameterName=Specular})
  ScalarParameter() {defaultValue=0.5, parameterName=SSSOpacity}
 ]
}
