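# Lit, Nanite-enabled material for a spinning fan in the BP13 front-end environment, using Garbage Island wind turbine textures.
# Time and a speed parameter drive RotateAboutAxis world position offset around a local-space axis and biased object pivot, masked by red vertex color.
# The same mask blends in rotation-corrected normals combined with the normal map's XY.
# Unconnected nodes remain for actor-position phase offset and alternative normal blending.
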Material /BRCosmetics/Environments/FrontEnd/Materials/BP13/M_Fan_FrontendBP13
graph {
 properties {bUseMaterialAttributes=true, bUsedWithNanite=true, pixelDepthOffsetMode=PDOM_Legacy, usedShadingModels=Default Lit}
 nodes [
  n1 = Frac(None=Multiply(A=Time() {period=5}, B=ScalarParameter() {defaultValue=0.2, group=FanRotation, parameterName=FanRotationSpeed}))
  n2 = ActorPositionWS()
  n3 = Constant() {r=1}
  Add(A=@n1, B=Frac(None=DotProduct(A=@n2, B=@n3)))
  n4 = Normalize(VectorInput=Transform(None=VectorParameter() {group=FanRotation, parameterName=NormalizedRotationAxis, defaultValue=<rgba 1 0 0 0>}.RGB) {transformSourceType=TRANSFORMSOURCE_Local})
  n5 = MaterialFunctionCall(Rotation Axis=@n4, Rotation Angle=@n1) {materialFunction=&FixRotateAboutAxisNormals}
  n6 = TextureSample() {samplerType=SAMPLERTYPE_Masks, texture=/Game/Environments/Apollo/Sets/GarbageIsland/Props/Textures/T_GarbageIsland_WindTurbine_01_S}
  n7 = TextureSample() {samplerType=SAMPLERTYPE_Normal, texture=/Game/Environments/Apollo/Sets/GarbageIsland/Props/Textures/T_GarbageIsland_WindTurbine_01_N}
  n8 = GetMaterialAttributes(None=MakeMaterialAttributes!(
    BaseColor=TextureSample() {texture=/Game/Environments/Apollo/Sets/GarbageIsland/Props/Textures/T_GarbageIsland_WindTurbine_01_D}.RGB
    Metallic=@n6.G
    Specular=@n6.G
    Roughness=@n6.G
    Anisotropy
    EmissiveColor
    Opacity
    OpacityMask
    Normal=@n7.RGB
    Tangent
    WorldPositionOffset
    SubsurfaceColor
    ClearCoat
    ClearCoatRoughness
    AmbientOcclusion
    Refraction
    CustomizedUVs_0
    CustomizedUVs_1
    CustomizedUVs_2
    CustomizedUVs_3
    CustomizedUVs_4
    CustomizedUVs_5
    CustomizedUVs_6
    CustomizedUVs_7
    PixelDepthOffset
    ShadingModel
    Displacement
   )) {attributeGetTypes [#1]}
  Normalize(VectorInput=MaterialFunctionCall(BaseNormal=@n5.New Tangent Space Vertex Normal, AdditionalNormal=@n8.Normal) {materialFunction=&BlendAngleCorrectedNormals}.Result)
  n9 = VertexColor()
  n10 = Add(
   A=MaterialFunctionCall() {materialFunction=&ObjectPivotPoint}.Object Pivot Location
   B=VectorParameter() {group=FanRotation, parameterName=PivotPointBias}.RGB
  )
  n11 = WorldPosition()
  n12 = Multiply(A=@n8.Normal, B=Constant3Vector() {constant=<rgba 1 1 0 0>})
  SetMaterialAttributes!(
   MaterialAttributes=@n8.MaterialAttributes
   World Position Offset=Multiply(A=@n9.R, B=RotateAboutAxis(NormalizedRotationAxis=@n4, RotationAngle=@n1, PivotPoint=@n10, Position=@n11.XYZ))
   Normal=LinearInterpolate(A=@n8.Normal, B=Add(A=@n5.New Tangent Space Vertex Normal, B=@n12), Alpha=@n9.R)
  ) {attributeSetTypes [#2, #1]}
  n13 = RotateAboutAxis(NormalizedRotationAxis=@n4, RotationAngle=@n1, PivotPoint=Constant(), Position=VertexNormalWS())
  Transform(None=Normalize(VectorInput=Add(A=@n8.Normal, B=@n13))) {transformSourceType=TRANSFORMSOURCE_World, transformType=TRANSFORM_Tangent}
  VertexInterpolator(VS=@n5.New Tangent Space Vertex Normal)
 ]
}
