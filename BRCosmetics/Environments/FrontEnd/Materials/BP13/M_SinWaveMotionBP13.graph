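# Lit material for Battle Royale front-end props that builds sine-driven swinging motion, with coastal buoy color, specular and normal textures as defaults.
# Time, offset per object by a hash of world position X and Z, drives two sine swings with parameterized axis, pivot, speed and angle.
# Vertex red masks the first rotation.
# The second, masked by green, rotates the already displaced position, and both offsets are summed for chained motion.
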Material /BRCosmetics/Environments/FrontEnd/Materials/BP13/M_SinWaveMotionBP13
graph {
 properties {
  bUsedWithMeshDeformer=true
  bUsedWithNanite=true
  bUsedWithSkeletalMesh=true
  pixelDepthOffsetMode=PDOM_Legacy
  usedShadingModels=Default Lit
 }
 nodes [
  n1 = Transform(None=VectorParameter() {parameterName=Axis, defaultValue=<rgba 1 0 0 1>}.RGB) {transformSourceType=TRANSFORMSOURCE_Local}
  n2 = MaterialFunctionCall(Float3=Divide(A=ObjectPositionWS()) {constB=1.33}) {materialFunction=&BreakOutFloat3Components}
  n3 = Add(A=Add(A=@n2.R, B=@n2.R), B=@n2.B)
  n4 = Add(A=Time(), B=Multiply(A=Frac(None=@n3), B=ScalarParameter() {defaultValue=24.324, parameterName=Random}))
  n5 = Sine(None=Multiply(A=@n4, B=ScalarParameter() {defaultValue=0.1, parameterName=SwingSpeed}))
  n6 = ScalarParameter() {defaultValue=0.1, parameterName=MaxRotation}
  n7 = VectorParameter() {parameterName=Pivot, defaultValue=<rgba 0 0 0 1>}
  n8 = Multiply(
   A=RotateAboutAxis(
    NormalizedRotationAxis=Normalize(VectorInput=@n1)
    RotationAngle=Multiply(A=@n5, B=@n6)
    PivotPoint=TransformPosition(None=@n7.RGB) {transformType=TRANSFORMPOSSOURCE_World}
    Position=WorldPosition().XYZ
   )
   B=VertexColor().R
  )
  n9 = Normalize(VectorInput=Transform(None=VectorParameter() {parameterName=Axis2, defaultValue=<rgba 1 0 0 1>}.RGB) {transformSourceType=TRANSFORMSOURCE_Local})
  n10 = Add(A=@n4, B=ScalarParameter() {parameterName=TimeOffset2})
  n11 = ScalarParameter() {defaultValue=0.1, parameterName=SwingSpeed2}
  n12 = Multiply(A=Sine(None=Multiply(A=@n10, B=@n11)), B=ScalarParameter() {defaultValue=0.1, parameterName=MaxRotation2})
  n13 = TransformPosition(None=VectorParameter() {parameterName=Pivot2, defaultValue=<rgba 0 0 0 1>}.RGB) {transformType=TRANSFORMPOSSOURCE_World}
  n14 = Add(A=@n8, B=WorldPosition().XYZ)
  Add(A=@n8, B=Multiply(A=RotateAboutAxis(NormalizedRotationAxis=@n9, RotationAngle=@n12, PivotPoint=@n13, Position=@n14), B=VertexColor().G))
  TextureSampleParameter2D() {parameterName=Color, texture=/Game/Environments/Apollo/Sets/Coastal/Props/Textures/T_Coastal_Buoy_01_D}
  TextureSampleParameter2D() {parameterName=Spec, samplerType=SAMPLERTYPE_Masks, texture=/Game/Environments/Apollo/Sets/Coastal/Props/Textures/T_Coastal_Buoy_01_S}
  TextureSampleParameter2D() {parameterName=Normal, samplerType=SAMPLERTYPE_Normal, texture=/Game/Environments/Apollo/Sets/Coastal/Props/Textures/T_Coastal_Buoy_01_N}
 ]
}
